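; BP_Industrial_TrommelGrp_B is a Fortnite industrial trommel group prop with hopper and conveyor meshes that follow the building's tint.
; The construction script calls its parent, then fetches the building texture data.
; If that array is not empty, it converts the first entry's tint to a linear color and passes it to SetColoronSecondaryMesh for both meshes.
; BeginPlay, overlap and Tick events are empty.
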(event EventBeginPlay)

(event Collision|EventActorBeginOverlap
  (params (OtherActor :type Actor%20Object%20Reference)))

(event EventTick
  (params (DeltaSeconds :type Float%20%28single-precision%29)))

(fn ConstructionScript
  (|CallParentFunction)
  (bind _outtexturedataarray (Building|GetBuildingTextureData))
  (if (> (Utilities|Array|Length _outtexturedataarray) 0)
    (for _array_element (Utilities|Array|MakeArray (Variables|Default|GetSM_Industrial_Hopper_A) (Variables|Default|GetSM_Industrial_Conveyor_A))
      (Default|SetColoronSecondaryMesh (Math|Conversions|ToLinearColor(Color) (Variables|/Script/FortniteGame.BuildingTextureData|GetTintColor (Utilities|Array|Get(acopy) _outtexturedataarray))) _array_element))))
