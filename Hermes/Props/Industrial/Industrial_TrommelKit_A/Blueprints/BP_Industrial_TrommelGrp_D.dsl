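; Industrial trommel prop blueprint (group D) that tints its static mesh components from the building's texture data.
; The construction script calls the parent, fetches the texture data array and takes the slot 0 tint color.
; It passes that color to each static mesh component via SetColoronSecondaryMesh, which per its comment writes custom primitive data.
; An empty array prints "TD Empty"; BeginPlay, overlap and Tick events are empty.
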(event EventBeginPlay)

(event Collision|EventActorBeginOverlap
  (params (OtherActor :type Actor%20Object%20Reference)))

(event EventTick
  (params (DeltaSeconds :type Float%20%28single-precision%29)))

(fn ConstructionScript
  (|CallParentFunction)
  ; Checks for a valid texturedata asset in Slot 0 (Primary Slot)
  (bind _outtexturedataarray (Building|GetBuildingTextureData))
  (if (> (Utilities|Array|Length _outtexturedataarray) 0)
    ; Goes through each static mesh component, and assigns tint color via custom prim data
    (for _array_element (Actor|GetComponentsByClass "/Script/Engine.StaticMeshComponent")
      (Default|SetColoronSecondaryMesh (Math|Conversions|ToLinearColor(Color) (Variables|/Script/FortniteGame.BuildingTextureData|GetTintColor (Utilities|Array|Get(acopy) _outtexturedataarray))) _array_element))
    (else
      (Development|PrintString "TD Empty"))))
